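# Prototype widget for one entry in the mobile Easy Loot Grid, built on FortMobileEasyLootGridListEntry, showing item icon, name and rarity.
# An invisible pickup button overlays the item info, and an initially collapsed side box holds a consume button.
# OnItemAdded plays a short scale pop-in and clears bAnimItemAdded.
# OnItemRemoved plays the scale-out, then calls UpdateListPostAnimation on the list when it finishes.
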WidgetBlueprint /Game/Athena/HUD/Bacchus/Proto_1/EasyLootGrid/BacchusEasyLootGridListEntry_Proto
animations [
 EasyLootGrid_Item_In {
  binding_count=1
  length_seconds=0.15
  object=&self:EasyLootGrid_Item_In
  sequence {bindings [HorizontalBox_0 {
     id=#1
     possessed_class=HorizontalBox
     tracks [{
       type=MovieScene2DTransformTrack
       sections [{
         type=MovieScene2DTransformSection
         end_frame=3
         start_frame=0
         channels [
          Scale.X {type=MovieSceneScriptingFloatChannel, keys [[0,0.2],[2,1.1],[3,1]]}
          Scale.Y {type=MovieSceneScriptingFloatChannel, keys [[0,0.2],[2,1.1],[3,1]]}
         ]
        }]
      }]
    }]}
 }
 EasyLootGrid_Item_Out {
  binding_count=1
  length_seconds=0.15
  object=&self:EasyLootGrid_Item_Out
  sequence {bindings [HorizontalBox_0 {
     id=#1
     possessed_class=HorizontalBox
     tracks [{
       type=MovieScene2DTransformTrack
       sections [{
         type=MovieScene2DTransformSection
         end_frame=3
         start_frame=0
         channels [
          Scale.X {type=MovieSceneScriptingFloatChannel, keys [[0,1],[1,1.1],[3,0.2]]}
          Scale.Y {type=MovieSceneScriptingFloatChannel, keys [[0,1],[1,1.1],[3,0.2]]}
         ]
        }]
      }]
    }]}
 }
]
graphs [EventGraph {nodes [
   n1 = Event OnItemRemoved!() -> (OutputDelegate, then, list:Fort Mobile Easy Loot Grid Object Reference) {is=title K2Node_Event}
   > StopAnimation(InAnimation=Get EasyLootGrid_Item_In() -> (EasyLootGrid_Item_In:Widget Animation Object Reference))
   > n2 = "Play Animation Time Range with Finished event (legacy)"!(
    execute:Exec
    Widget:User Widget Object Reference=Self-Reference()
    InAnimation:Widget Animation Object Reference=Get EasyLootGrid_Item_Out() -> (EasyLootGrid_Item_Out:Widget Animation Object Reference)
    StartAtTime:float=0
    EndAtTime:float=0
    NumLoopsToPlay:int=1
    PlayMode:EUMGSequencePlayMode Enum
    PlaybackSpeed:float=1
   ) -> (then:Exec, Result:UMGSequence Player Object Reference, Finished:Exec) {is=title K2Node_PlayAnimationTimeRange}
   UpdateListPostAnimation(execute=@n2.Finished, self:Fort Mobile Easy Loot Grid Object Reference=@n1.list)
   Event OnItemAdded() {is=title K2Node_Event}
   > StopAnimation(InAnimation=Get EasyLootGrid_Item_Out())
   > PlayAnimationForward(InAnimation=Get EasyLootGrid_Item_In())
   > Set bAnimItemAdded(bAnimItemAdded:bool=false) -> (Output_Get:bool)
  ]}]
variables []
widgets {
 info {inheritedWidgetCount=6, parentClass=/Script/FortniteUI.FortMobileEasyLootGridListEntry, rootWidgetClass=&SizeBox, widgetCount=12}
 widgets [
  SizeBox_41 {widgetClassPath=&SizeBox, widget=SizeBox_41, properties {bOverride_HeightOverride=true, heightOverride=100}}
  HorizontalBox_0 {widgetClassPath=&HorizontalBox, parent=SizeBox_41, widget=HorizontalBox_0, slot {class=SizeBoxSlot}}
  All {
   widgetClassPath=&Overlay
   parent=HorizontalBox_0
   widget=All
   slot {class=HorizontalBoxSlot, properties {size {sizeRule=Fill, value=1}}}
  }
  HorizontalBox_Info {
   widgetClassPath=&HorizontalBox
   parent=All
   widget=HorizontalBox_Info
   slot {class=OverlaySlot, properties {horizontalAlignment=HAlign_Fill, verticalAlignment=VAlign_Fill, padding=<ltrb 14 0 0 0>}}
  }
  SizeBox_Icon {
   widgetClassPath=&SizeBox
   parent=HorizontalBox_Info
   widget=SizeBox_Icon
   properties {bOverride_WidthOverride=true, widthOverride=100}
   slot {class=HorizontalBoxSlot}
  }
  ItemIcon {
   widgetClassPath=/Game/Athena/HUD/Inventory/AthenaItemIconLoadout
   bInherited=true
   bIsVariable=true
   parent=SizeBox_Icon
   widget=ItemIcon
   properties {bDisplayInitialAmmoCount=true, bShowItemTier=false}
   slot {class=SizeBoxSlot}
  }
  VerticalBox_Text {
   widgetClassPath=&VerticalBox
   parent=HorizontalBox_Info
   widget=VerticalBox_Text
   slot {class=HorizontalBoxSlot, properties {verticalAlignment=VAlign_Center, padding=<ltrb 12 0 12 0>}}
  }
  Text_WeaponName {
   widgetClassPath=&CommonTextBlock
   bInherited=true
   parent=VerticalBox_Text
   widget=Text_WeaponName
   properties {
    text=_Item Name_
    font {
     bForceMonospaced=false
     bMaterialIsStencil=false
     fontMaterial=/Game/Athena/HUD/Inventory/Art/M_UI_Inventory_Loadout_FontColor
     fontObject=&BurbankSmall
     letterSpacing=0
     monospacedWidth=1
     size=20
     skewAmount=0
     typefaceFontName=Black
     outlineSettings {
      bApplyOutlineToDropShadows=false
      bMiteredCorners=false
      bSeparateFillAlpha=false
      outlineMaterial=null
      outlineSize=0
      outlineColor=<rgba 0 0 0 1>
     }
    }
    shadowOffset=<0 0>
   }
   slot {class=VerticalBoxSlot, properties {horizontalAlignment=HAlign_Left, verticalAlignment=VAlign_Center}}
  }
  Text_WeaponRarity {
   widgetClassPath=&CommonTextBlock
   bInherited=true
   parent=VerticalBox_Text
   widget=Text_WeaponRarity
   properties {
    text=_Rarity_
    textTransformPolicy=ToUpper
    font {
     bForceMonospaced=false
     bMaterialIsStencil=false
     fontMaterial=&M_UI_Inventory_Loadout_FontColor
     fontObject=&BurbankSmall
     letterSpacing=0
     monospacedWidth=1
     size=14
     skewAmount=0
     typefaceFontName=Black
     outlineSettings {
      bApplyOutlineToDropShadows=false
      bMiteredCorners=false
      bSeparateFillAlpha=false
      outlineMaterial=null
      outlineSize=0
      outlineColor=<rgba 0 0 0 1>
     }
    }
    shadowOffset=<0 0>
   }
   slot {class=VerticalBoxSlot, properties {horizontalAlignment=HAlign_Left, verticalAlignment=VAlign_Center}}
  }
  Button_PickUp {
   widgetClassPath=&/BP_EAsyLootGridButton
   bInherited=true
   bIsVariable=true
   parent=All
   widget=Button_PickUp
   properties {holdData=&CommonUI_HoldData_Base_C, colorAndOpacity=<rgba 0 0 0 0>}
   slot {class=OverlaySlot, properties {horizontalAlignment=HAlign_Fill, verticalAlignment=VAlign_Fill}}
  }
  Box_SecondaryAbility {
   widgetClassPath=&SizeBox
   bInherited=true
   parent=HorizontalBox_0
   widget=Box_SecondaryAbility
   properties {bOverride_HeightOverride=true, bOverride_WidthOverride=true, heightOverride=86, visibility=Collapsed, widthOverride=86}
   slot {class=HorizontalBoxSlot, properties {horizontalAlignment=HAlign_Center, verticalAlignment=VAlign_Center}}
  }
  Button_Consume {
   widgetClassPath=&WBP_UIKit_Button_Generic
   bInherited=true
   bIsVariable=true
   parent=Box_SecondaryAbility
   widget=Button_Consume
   properties {
    backgroundMaterial=null
    iconMaterial=/Game/Athena/HUD/Bacchus/Proto_1/EasyLootGrid/Art/MI_EasyLootGrid_SecondaryAction
    useBackgroundBlock=false
    useTextBlock=false
    buttonPadding=<ltrb 0 0 0 0>
    colorAndOpacity=<rgba 1 1 1 0.6>
    iconSize=<86 86>
   }
   slot {class=SizeBoxSlot}
  }
 ]
}
