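# HUD Layout Tool preview widget for the Easy Loot Grid, showing a preview image on a translucent backing.
# On PreConstruct it casts the owning player to FortPlayerController, then sets itself Visible if IsEasyLootGridEnabled or design time is true, otherwise Collapsed.
# The layout is a user-specified ScaleBox around a fixed-height SizeBox, with the inherited instances-left counter overlaid at bottom right.
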WidgetBlueprint /Game/Athena/HUD/Bacchus/Proto_1/EasyLootGrid/WidgetPreview_EasyLootGrid
graphs [EventGraph {nodes [
   n1 = Event PreConstruct()
   > n2 = Cast To FortPlayerController(Object=GetOwningPlayer())
   n3 = IsEasyLootGridEnabled!(FortPC:Fort Player Controller Object Reference=@n2) -> (ReturnValue:bool)
   > SetVisibility(InVisibility=Select(Option 0=Collapsed, Option 1=Visible, Index=OR Boolean(A=@n3, B=@n1.IsDesignTime)))
  ]}]
variables []
widgets {
 info {inheritedWidgetCount=2, parentClass=/Script/FortniteUI.HUDLayoutToolV2_WidgetPreview, rootWidgetClass=&ScaleBox, widgetCount=7}
 widgets [
  ScaleBox_0 {widgetClassPath=&ScaleBox, widget=ScaleBox_0, properties {stretch=UserSpecified}}
  LootGridVisibilityBox {
   widgetClassPath=&SizeBox
   parent=ScaleBox_0
   widget=LootGridVisibilityBox
   properties {bOverride_HeightOverride=true, heightOverride=128, widthOverride=100}
   slot {class=ScaleBoxSlot}
  }
  Overlay_250 {widgetClassPath=&Overlay, parent=LootGridVisibilityBox, widget=Overlay_250, slot {class=SizeBoxSlot}}
  Border_85 {
   widgetClassPath=&Border
   parent=Overlay_250
   widget=Border_85
   properties {
    horizontalAlignment=HAlign_Center
    verticalAlignment=VAlign_Center
    background {
     bIsDynamicallyLoaded=false
     drawAs=Image
     imageType=NoImage
     mirroring=NoMirror
     resourceName=None
     resourceObject=null
     tiling=NoTile
     imageSize=<32 32>
     margin=<ltrb 0 0 0 0>
     outlineSettings {
      bUseBrushTransparency=false
      roundingType=HalfHeightRadius
      width=0
      color {colorUseRule=UseColor_Specified, specifiedColor=<rgba 0 0 0 0>}
      cornerRadii=<0 0 0 0>
     }
     tintColor {colorUseRule=UseColor_Specified, specifiedColor=<rgba 0 0 0 0.2>}
     uVRegion {bIsValid=false, max=<0 0>, min=<0 0>}
    }
    padding=<ltrb 20 20 20 20>
   }
   slot {class=OverlaySlot, properties {horizontalAlignment=HAlign_Fill, verticalAlignment=VAlign_Fill}}
  }
  Image_EasyLootGrid_HLTPreview {
   widgetClassPath=&Image
   bIsVariable=true
   parent=Border_85
   widget=Image_EasyLootGrid_HLTPreview
   properties {brush {
     bIsDynamicallyLoaded=false
     drawAs=Image
     imageType=NoImage
     mirroring=NoMirror
     resourceName=None
     resourceObject=&/EasyLootGrid_HLT_Preview
     tiling=NoTile
     imageSize=<78 89>
     margin=<ltrb 0 0 0 0>
     outlineSettings {
      bUseBrushTransparency=false
      roundingType=HalfHeightRadius
      width=0
      color {colorUseRule=UseColor_Specified, specifiedColor=<rgba 0 0 0 0>}
      cornerRadii=<0 0 0 0>
     }
     tintColor {colorUseRule=UseColor_Specified, specifiedColor=<rgba 1 1 1 1>}
     uVRegion {bIsValid=false, max=<0 0>, min=<0 0>}
    }}
   slot {class=BorderSlot, properties {horizontalAlignment=HAlign_Center, verticalAlignment=VAlign_Center, padding=<ltrb 20 20 20 20>}}
  }
  TextBlock_InstancesLeftCounter {
   widgetClassPath=&CommonTextBlock
   bInherited=true
   parent=Overlay_250
   widget=TextBlock_InstancesLeftCounter
   properties {
    justification=Center
    textTransformPolicy=ToUpper
    font {
     bForceMonospaced=false
     bMaterialIsStencil=false
     fontMaterial=null
     fontObject=&BurbankSmall
     letterSpacing=0
     monospacedWidth=1
     size=16
     skewAmount=0
     typefaceFontName=Black
     outlineSettings {
      bApplyOutlineToDropShadows=false
      bMiteredCorners=false
      bSeparateFillAlpha=false
      outlineMaterial=null
      outlineSize=0
      outlineColor=<rgba 0 0 0 1>
     }
    }
    shadowOffset=<0 0>
   }
   slot {class=OverlaySlot, properties {horizontalAlignment=HAlign_Right, verticalAlignment=VAlign_Bottom, padding=<ltrb 0 0 4 0>}}
  }
  Border_Background {widgetClassPath=&Border, bInherited=true, widget=null, properties=~}
 ]
}
